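# Material function that outputs a distance-based fade factor for terrain, measured from the camera to the world position of the shaded pixel.
# It subtracts the CameraBlendOffset parameter from that distance, divides the result by CameraBlendDistance and clamps it to form a linear blend mask.
# Both the offset and blend range are scalar parameters, so materials using the function can tune where the fade starts and how long it runs.
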MaterialFunction /FigmentEnvironment/Migration51/Athena/Environments/Landscape/MaterialFunctions/MF_TerrainDistanceFade
graph {nodes [
  n1 = Subtract(A=Distance(A=WorldPosition().XYZ, B=CameraPositionWS()), B=ScalarParameter() {defaultValue=2000, parameterName=CameraBlendOffset})
  n2 = ScalarParameter() {defaultValue=55000, parameterName=CameraBlendDistance}
  FunctionOutput(None=Clamp(None=Divide(A=@n1, B=@n2)))
 ]}
